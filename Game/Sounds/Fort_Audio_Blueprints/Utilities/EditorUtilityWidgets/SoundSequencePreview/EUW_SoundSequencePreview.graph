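# Editor utility widget that previews a Fort Sound Sequence, picked through a property view, with Play and Stop buttons.
# Play enables subtitles and spawns each item as a 2D sound in the first PIE world, falling back to the editor world.
# Audio-finished callbacks and per-item delay timers chain the items by index.
# Stop unbinds and halts the audio and resets the index, while only Play clears a pending delay timer.
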EditorUtilityWidgetBlueprint /Game/Sounds/Fort_Audio_Blueprints/Utilities/EditorUtilityWidgets/SoundSequencePreview/EUW_SoundSequencePreview
graphs [
 AdvanceSequence {nodes [
   AdvanceSequence() {is=title K2Node_FunctionEntry}
   n1 = GetSoundSequenceData!(self:Fort Sound Sequence Object Reference=Get Sequence() -> (Sequence:Fort Sound Sequence Object Reference)) -> (OutSoundSequenceData:Array of Sound Sequence Data Structures)
   > n2 = Branch(Condition="Equal (Integer)"(A=Get CurrentItemIndex() -> (CurrentItemIndex:int=0), B=Length(TargetArray:Array of Sound Sequence Data Structures=@n1)))
   > Set CurrentItemIndex(CurrentItemIndex:int=0) -> (Output_Get:int=0)
   Increment Int(" "=@n2.else, Value=Get CurrentItemIndex() -> (CurrentItemIndex=0))
   n3 = Get Sequence()
   > n4 = Branch(Condition=Is Valid Index!(
     TargetArray:Array of Sound Sequence Data Structures=GetSoundSequenceData(self=@n3)
     IndexToTest:int=Get CurrentItemIndex() -> (CurrentItemIndex=0)
    ) -> (ReturnValue:bool))
   > PlayCurrentItem()
   Set CurrentItemIndex(execute=@n4.else, CurrentItemIndex=0) -> (Output_Get=0)
  ]}
 EventGraph {nodes [
   "On Clicked (StopButton)"() {is=title K2Node_ComponentBoundEvent}
   > Stop()
   "On Clicked (PlayButton)"() {is=title K2Node_ComponentBoundEvent}
   > Play()
   Event Construct()
   n1 = Get CIDProperty() -> (CIDProperty:Single Property View Object Reference)
   > SetObject(self=@n1, NewObject=Self-Reference())
   > SetPropertyName!(execute, self:Single Property View Object Reference=@n1, NewPropertyName:Name=Sequence) -> (then)
  ]}
 GetCurrentItem {nodes [
   GetCurrentItem() {is=title K2Node_FunctionEntry}
   n1 = Get Sequence()
   > Return Node!(
    execute
    Item:Sound Sequence Data Structure="Get (a copy)"(Array:Array of Sound Sequence Data Structures=GetSoundSequenceData(self=@n1), Dimension 1=Get CurrentItemIndex() -> (CurrentItemIndex=0)) -> (Output:Sound Sequence Data Structure)
   )
  ]}
 GetWorld {nodes [
   GetWorld() {is=title K2Node_FunctionEntry}
   > n1 = GetPIEWorlds!(execute, bIncludeDedicatedServer:bool=false) -> (then, ReturnValue:Array of World Object References)
   > n2 = Branch(Condition=Is Valid Index(TargetArray:Array of World Object References=@n1))
   > Return Node!(execute, ReturnValue:World Object Reference="Get (a copy)"(Array:Array of World Object References=@n1) -> (Output:World Object Reference))
   n3 = GetEditorWorld(execute=@n2.else, self=Get UnrealEditorSubsystem())
   > Is Valid(InputObject=@n3)
   > Return Node(ReturnValue=@n3)
  ]}
 HandleDelay {nodes [
   HandleDelay() {is=title K2Node_FunctionEntry}
   n1 = GetCurrentItem!() -> (Item_Sound:Sound Base Object Reference, Item_Delay:float=~) {is=title K2Node_CallFunction}
   > n2 = Branch(Condition="float > float"(A=@n1.Item_Delay))
   > n3 = Set Timer by Event(Delegate=Create Event(), Time=@n1.Item_Delay, bLooping=false)
   > Set CurrentDelayTimer(CurrentDelayTimer:Timer Handle Structure=@n3) -> (Output_Get:Timer Handle Structure)
   AdvanceSequence(execute=@n2.else) {is=title K2Node_CallFunction}
  ]}
 Play {nodes [
   Play() {is=title K2Node_FunctionEntry}
   > SetSubtitlesEnabled!(execute, bEnabled:bool=true) -> (then)
   > PlayCurrentItem()
   > Clear and Invalidate Timer by Handle!(execute, WorldContextObject:Object Reference, Handle=Get CurrentDelayTimer() -> (CurrentDelayTimer:Timer Handle Structure)) -> (then)
  ]}
 PlayCurrentItem {nodes [
   PlayCurrentItem() {is=title K2Node_FunctionEntry}
   n1 = GetCurrentItem() -> (Item_Delay=~)
   > n2 = Is Valid(InputObject=@n1)
   Set CurrentItemIndex(execute=@n2.Is Not Valid, CurrentItemIndex=0) -> (Output_Get=0)
   n3 = SpawnSound2D!(
    execute=@n2.Is Valid
    WorldContextObject:Object Reference=GetWorld!() -> (ReturnValue:World Object Reference) {is=title K2Node_CallFunction}
    Sound:Sound Base Object Reference=@n1
    VolumeMultiplier:float=1
    PitchMultiplier:float=1
    StartTime:float=0
    ConcurrencySettings:Sound Concurrency Object Reference
    bPersistAcrossLevelTransition:bool=false
    bAutoDestroy:bool=true
   ) -> (then, ReturnValue:Audio Component Object Reference)
   > n4 = Set CurrentItemAudio(CurrentItemAudio:Audio Component Object Reference=@n3) -> (Output_Get:Audio Component Object Reference)
   > Bind Event to On Audio Finished(self:Audio Component Object Reference=@n4, Delegate=Create Event()) {is=title K2Node_AddDelegate}
  ]}
 Stop {nodes [
   Stop() {is=title K2Node_FunctionEntry}
   n1 = Get CurrentItemAudio() -> (CurrentItemAudio:Audio Component Object Reference)
   > Unbind Event from On Audio Finished(self:Audio Component Object Reference=@n1, Delegate=Create Event()) {is=title K2Node_RemoveDelegate}
   > Stop(self:Audio Component Object Reference=@n1) {is=title K2Node_CallFunction}
   > Set CurrentItemIndex(CurrentItemIndex=0) -> (Output_Get=0)
  ]}
]
variables [
 CurrentDelayTimer {type=TimerHandle}
 CurrentItemAudio {type=AudioComponent, default=null}
 CurrentItemIndex {type=integer}
 Sequence {type=FortSoundSequence, default=null}
]
widgets {
 info {rootWidgetClass=&CanvasPanel, widgetCount=13}
 widgets [
  CanvasPanel_0 {widgetClassPath=&CanvasPanel, widget=CanvasPanel_0}
  ScaleBox_262 {
   widgetClassPath=&ScaleBox
   parent=CanvasPanel_0
   widget=ScaleBox_262
   slot {class=CanvasPanelSlot, properties {layoutData {alignment=<0 0>, anchors {maximum=<0 0>, minimum=<0 0>}, offsets=<ltrb 5 5 400 100>}}}
  }
  VerticalBox_202 {
   widgetClassPath=&VerticalBox
   parent=ScaleBox_262
   widget=VerticalBox_202
   slot {class=ScaleBoxSlot, properties {horizontalAlignment=HAlign_Fill, verticalAlignment=VAlign_Fill}}
  }
  HorizontalBox_107 {
   widgetClassPath=&HorizontalBox
   parent=VerticalBox_202
   widget=HorizontalBox_107
   slot {class=VerticalBoxSlot, properties {size {sizeRule=Fill, value=0.8}}}
  }
  VerticalBox_208 {
   widgetClassPath=&VerticalBox
   parent=HorizontalBox_107
   widget=VerticalBox_208
   slot {class=HorizontalBoxSlot, properties {size {sizeRule=Fill, value=0.2}}}
  }
  CIDProperty {
   widgetClassPath=&SinglePropertyView
   parent=VerticalBox_208
   widget=CIDProperty
   properties {nameOverride=Sound Sequence, propertyName=Sequence}
   slot {class=VerticalBoxSlot, properties {size {sizeRule=Fill, value=1}}}
  }
  HorizontalBox_260 {
   widgetClassPath=&HorizontalBox
   parent=VerticalBox_202
   widget=HorizontalBox_260
   slot {class=VerticalBoxSlot, properties {size {sizeRule=Fill, value=0.2}}}
  }
  PlayButton {
   widgetClassPath=&Button
   parent=HorizontalBox_260
   widget=PlayButton
   properties {widgetStyle {
     clickedSlateSound {resourceObject=null}
     disabled {
      bIsDynamicallyLoaded=false
      drawAs=RoundedBox
      imageType=NoImage
      mirroring=NoMirror
      resourceName=None
      resourceObject=null
      tiling=NoTile
      imageSize=<0 0>
      margin=<ltrb 0 0 0 0>
      outlineSettings {
       bUseBrushTransparency=false
       roundingType=FixedRadius
       width=1
       color {colorUseRule=UseColor_ColorTable, specifiedColor=<rgba 0 0 0 0>}
       cornerRadii=<4 4 4 4>
      }
      tintColor {colorUseRule=UseColor_Specified, specifiedColor=<rgba 0.0395 0.0395 0.0395 1>}
      uVRegion {bIsValid=false, max=<0 0>, min=<0 0>}
     }
     disabledForeground {colorUseRule=UseColor_ColorTable, specifiedColor=<rgba 0 0 0 0>}
     hovered {
      bIsDynamicallyLoaded=false
      drawAs=RoundedBox
      imageType=NoImage
      mirroring=NoMirror
      resourceName=None
      resourceObject=null
      tiling=NoTile
      imageSize=<0 0>
      margin=<ltrb 0 0 0 0>
      outlineSettings {
       bUseBrushTransparency=false
       roundingType=FixedRadius
       width=1
       color {colorUseRule=UseColor_ColorTable, specifiedColor=<rgba 0 0 0 0>}
       cornerRadii=<4 4 4 4>
      }
      tintColor {colorUseRule=UseColor_Specified, specifiedColor=<rgba 0.0953 0.0953 0.0953 1>}
      uVRegion {bIsValid=false, max=<0 0>, min=<0 0>}
     }
     hoveredForeground {colorUseRule=UseColor_ColorTable, specifiedColor=<rgba 0 0 0 0>}
     hoveredSlateSound {resourceObject=null}
     normal {
      bIsDynamicallyLoaded=false
      drawAs=RoundedBox
      imageType=NoImage
      mirroring=NoMirror
      resourceName=None
      resourceObject=null
      tiling=NoTile
      imageSize=<0 0>
      margin=<ltrb 0 0 0 0>
      outlineSettings {
       bUseBrushTransparency=false
       roundingType=FixedRadius
       width=1
       color {colorUseRule=UseColor_ColorTable, specifiedColor=<rgba 0 0 0 0>}
       cornerRadii=<4 4 4 4>
      }
      tintColor {colorUseRule=UseColor_Specified, specifiedColor=<rgba 0.0395 0.0395 0.0395 1>}
      uVRegion {bIsValid=false, max=<0 0>, min=<0 0>}
     }
     normalForeground {colorUseRule=UseColor_ColorTable, specifiedColor=<rgba 0 0 0 0>}
     normalPadding=<ltrb 12 1.5 12 1.5>
     pressed {
      bIsDynamicallyLoaded=false
      drawAs=RoundedBox
      imageType=NoImage
      mirroring=NoMirror
      resourceName=None
      resourceObject=null
      tiling=NoTile
      imageSize=<0 0>
      margin=<ltrb 0 0 0 0>
      outlineSettings {
       bUseBrushTransparency=false
       roundingType=FixedRadius
       width=1
       color {colorUseRule=UseColor_ColorTable, specifiedColor=<rgba 0 0 0 0>}
       cornerRadii=<4 4 4 4>
      }
      tintColor {colorUseRule=UseColor_Specified, specifiedColor=<rgba 0.0284 0.0284 0.0284 1>}
      uVRegion {bIsValid=false, max=<0 0>, min=<0 0>}
     }
     pressedForeground {colorUseRule=UseColor_ColorTable, specifiedColor=<rgba 0 0 0 0>}
     pressedPadding=<ltrb 12 2 12 0>
     pressedSlateSound {resourceObject=null}
    }}
   slot {class=HorizontalBoxSlot, properties {horizontalAlignment=HAlign_Left, size {sizeRule=Automatic, value=0.5}}}
  }
  ScaleBox_494 {
   widgetClassPath=&ScaleBox
   parent=PlayButton
   widget=ScaleBox_494
   slot {class=ButtonSlot, properties {horizontalAlignment=HAlign_Fill, verticalAlignment=VAlign_Fill, padding=<ltrb 0 0 0 0>}}
  }
  CommonTextBlock_750 {
   widgetClassPath=&CommonTextBlock
   parent=ScaleBox_494
   widget=CommonTextBlock_750
   properties {
    style=&TextStyle-Base_C
    text=Play
    font {
     bForceMonospaced=false
     bMaterialIsStencil=false
     fontMaterial=null
     fontObject=&NotoSans
     letterSpacing=0
     monospacedWidth=1
     size=24
     skewAmount=0
     typefaceFontName=NotoSans-Regular
     outlineSettings {
      bApplyOutlineToDropShadows=false
      bMiteredCorners=false
      bSeparateFillAlpha=false
      outlineMaterial=null
      outlineSize=0
      outlineColor=<rgba 0 0 0 1>
     }
    }
    renderTransform {angle=0, scale=<1 1>, shear=<0 0>, translation=<0 -2.5>}
    shadowOffset=<0 0>
   }
   slot {class=ScaleBoxSlot}
  }
  StopButton {
   widgetClassPath=&Button
   parent=HorizontalBox_260
   widget=StopButton
   properties {widgetStyle {
     clickedSlateSound {resourceObject=null}
     disabled {
      bIsDynamicallyLoaded=false
      drawAs=RoundedBox
      imageType=NoImage
      mirroring=NoMirror
      resourceName=None
      resourceObject=null
      tiling=NoTile
      imageSize=<0 0>
      margin=<ltrb 0 0 0 0>
      outlineSettings {
       bUseBrushTransparency=false
       roundingType=FixedRadius
       width=1
       color {colorUseRule=UseColor_ColorTable, specifiedColor=<rgba 0 0 0 0>}
       cornerRadii=<4 4 4 4>
      }
      tintColor {colorUseRule=UseColor_Specified, specifiedColor=<rgba 0.0395 0.0395 0.0395 1>}
      uVRegion {bIsValid=false, max=<0 0>, min=<0 0>}
     }
     disabledForeground {colorUseRule=UseColor_ColorTable, specifiedColor=<rgba 0 0 0 0>}
     hovered {
      bIsDynamicallyLoaded=false
      drawAs=RoundedBox
      imageType=NoImage
      mirroring=NoMirror
      resourceName=None
      resourceObject=null
      tiling=NoTile
      imageSize=<0 0>
      margin=<ltrb 0 0 0 0>
      outlineSettings {
       bUseBrushTransparency=false
       roundingType=FixedRadius
       width=1
       color {colorUseRule=UseColor_ColorTable, specifiedColor=<rgba 0 0 0 0>}
       cornerRadii=<4 4 4 4>
      }
      tintColor {colorUseRule=UseColor_Specified, specifiedColor=<rgba 0.0953 0.0953 0.0953 1>}
      uVRegion {bIsValid=false, max=<0 0>, min=<0 0>}
     }
     hoveredForeground {colorUseRule=UseColor_ColorTable, specifiedColor=<rgba 0 0 0 0>}
     hoveredSlateSound {resourceObject=null}
     normal {
      bIsDynamicallyLoaded=false
      drawAs=RoundedBox
      imageType=NoImage
      mirroring=NoMirror
      resourceName=None
      resourceObject=null
      tiling=NoTile
      imageSize=<0 0>
      margin=<ltrb 0 0 0 0>
      outlineSettings {
       bUseBrushTransparency=false
       roundingType=FixedRadius
       width=1
       color {colorUseRule=UseColor_ColorTable, specifiedColor=<rgba 0 0 0 0>}
       cornerRadii=<4 4 4 4>
      }
      tintColor {colorUseRule=UseColor_Specified, specifiedColor=<rgba 0.0395 0.0395 0.0395 1>}
      uVRegion {bIsValid=false, max=<0 0>, min=<0 0>}
     }
     normalForeground {colorUseRule=UseColor_ColorTable, specifiedColor=<rgba 0 0 0 0>}
     normalPadding=<ltrb 12 1.5 12 1.5>
     pressed {
      bIsDynamicallyLoaded=false
      drawAs=RoundedBox
      imageType=NoImage
      mirroring=NoMirror
      resourceName=None
      resourceObject=null
      tiling=NoTile
      imageSize=<0 0>
      margin=<ltrb 0 0 0 0>
      outlineSettings {
       bUseBrushTransparency=false
       roundingType=FixedRadius
       width=1
       color {colorUseRule=UseColor_ColorTable, specifiedColor=<rgba 0 0 0 0>}
       cornerRadii=<4 4 4 4>
      }
      tintColor {colorUseRule=UseColor_Specified, specifiedColor=<rgba 0.0284 0.0284 0.0284 1>}
      uVRegion {bIsValid=false, max=<0 0>, min=<0 0>}
     }
     pressedForeground {colorUseRule=UseColor_ColorTable, specifiedColor=<rgba 0 0 0 0>}
     pressedPadding=<ltrb 12 2 12 0>
     pressedSlateSound {resourceObject=null}
    }}
   slot {class=HorizontalBoxSlot, properties {horizontalAlignment=HAlign_Left, size {sizeRule=Automatic, value=0.5}}}
  }
  ScaleBox {
   widgetClassPath=&ScaleBox
   parent=StopButton
   widget=ScaleBox
   slot {class=ButtonSlot, properties {horizontalAlignment=HAlign_Fill, verticalAlignment=VAlign_Fill, padding=<ltrb 0 0 0 0>}}
  }
  CommonTextBlock {
   widgetClassPath=&CommonTextBlock
   parent=ScaleBox
   widget=CommonTextBlock
   properties {
    style=&TextStyle-Base_C
    text=Stop
    font {
     bForceMonospaced=false
     bMaterialIsStencil=false
     fontMaterial=null
     fontObject=&NotoSans
     letterSpacing=0
     monospacedWidth=1
     size=24
     skewAmount=0
     typefaceFontName=NotoSans-Regular
     outlineSettings {
      bApplyOutlineToDropShadows=false
      bMiteredCorners=false
      bSeparateFillAlpha=false
      outlineMaterial=null
      outlineSize=0
      outlineColor=<rgba 0 0 0 1>
     }
    }
    renderTransform {angle=0, scale=<1 1>, shear=<0 0>, translation=<0 -2.5>}
    shadowOffset=<0 0>
   }
   slot {class=ScaleBoxSlot}
  }
 ]
}
